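# Gold-variant Niagara burst for the swap-to-lion moment of the FeatherMudLounge emote.
# Skeletal-mesh-sampled mesh flames and embers combine with speed lines, a shockwave mesh, delayed flare, sprite decal and light-casting glow, all short-lived.
# DarkGlow and MeshFlames use spawn groups as visibility tags, giving two differently styled layers per emitter.
# Extra per-renderer attributes drive glow, flash and light layers, with one ember layer's alpha adjusted for front end versus gameplay.
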NiagaraSystem /BRCosmetics/Animation/Game/MainPlayer/Emotes/FeatherMudLounge/FX/Gold/NS_Emote_FeatherMudLounge_SwapToLion_Burst_Gold
dependencies {
 usedDataInterfaces [
  &NiagaraDataInterfaceColorCurve
  &NiagaraDataInterfaceCurve
  &NiagaraDataInterfaceVectorCurve
  &NiagaraDataInterfaceSkeletalMesh
 ]
 usedDynamicInputs []
 usedModules [
  &self:SystemSpawnScript.SystemScriptSource.SystemScriptGraph.NiagaraNodeAssignment_1.SetVariables_#1
  &SystemState
 ]
}
emitters {
 DarkGlow {
  emitterUpdateScript [
   &EmitterState()
   &SpawnBurst_Instantaneous(Spawn Group=1, Use Spawn Probability=false, UseLoopCountLimit=false) {module=SpawnBurst_Instantaneous001}
   &SpawnBurst_Instantaneous(Use Spawn Probability=false, UseLoopCountLimit=false)
  ]
  particleSpawnScript [
   /SetVariables(Particles.VisibilityTag=&SpawnGroup()) {isSetParameters=true}
   &InitializeParticle(
    Color=&MaskLinearColorBySpawnGroup(Masked Linear Color=<rgba 0.3177 0.0647 0.0075 4>, Passthrough Linear Color=<rgba 10 6.2769 0.1873 1.25>)
    Lifetime=0.3
    Position Offset=&MaskVectorBySpawnGroup(Masked Vector=<0 0 -20>, Passthrough Vector=<0 0 0>)
    Sprite Size Mode=::NewEnumerator3
    Uniform Sprite Size=&MaskFloatBySpawnGroup(Masked Float=250, Passthrough Float=170)
    UsePositionOffset=true
    Write Material Random=true
   )
  ]
  particleUpdateScript [
   &ParticleState()
   &ScaleSpriteSize(
    Scale Sprite Size Mode=::NewEnumerator0
    Uniform Scale Factor=&MaskFloatBySpawnGroup(
     Masked Float=1
     Passthrough Float=&FloatFromCurve(FloatCurve=DataInterface() {propertyValues {Curve {keys [
          {interpMode=RCIM_Cubic, leaveTangent=2.4364, tangentMode=RCTM_Break, time=0, value=0}
          {interpMode=RCIM_Cubic, leaveTangent=1.1981, tangentMode=RCTM_Break, time=0.3, value=0.75}
          <at 1 1.1>
         ]}}})
    )
   )
   &ScaleColor(
    Scale Mode=::NewEnumerator1
    Scale RGBA=&MakeVector4_FromLinearColor(LinearColor=&MaskLinearColorBySpawnGroup(
      Masked Linear Color=&ColorFromCurve(ColorCurve=DataInterface() {propertyValues {
         AlphaCurve {keys [<at 0 0>,<at 0.1573 1>,<at 0.5151 0.3>,<at 1 0>]}
         BlueCurve {keys [<at 0 1>]}
         GreenCurve {keys [<at 0 1>]}
         RedCurve {keys [<at 0 1>]}
        }})
      Passthrough Linear Color=&MakeLinearColorFromVectorAndFloat(
       "Vector (RGB)"=<1 1 1>
       "Float (Alpha)"=&FloatFromCurve(
        FloatCurve=DataInterface() {propertyValues {Curve {keys [{interpMode=RCIM_Cubic, leaveTangent=-2.9394, tangentMode=RCTM_Break, time=0, value=1}, <at 1 0>]}}}
        Scale Curve=0.8
       )
      )
     ))
    ScaleRGBA=true
   )
   &CameraOffset(Camera Offset Amount=&MaskFloatBySpawnGroup(Masked Float=-30, Passthrough Float=25))
   &FN_Light_Attributes(
    Color=<rgba 0.2896 1.0286 2 3>
    Create Color Attribute=true
    Create Exponent Attribute=true
    Create Position Attribute=true
    Create Radius Attribute=true
    Exponent=2.5
    Position=&AddVectorToPosition(Position=$Particles.Initial.Position, Vector=<0 0 -80>)
    Radius=220
    Scale Color=&FloatFromCurve(
     FloatCurve=DataInterface() {propertyValues {Curve {keys [
         {interpMode=RCIM_Cubic, leaveTangent=10.2442, tangentMode=RCTM_Break, time=0, value=0.15}
         {interpMode=RCIM_Cubic, tangentMode=RCTM_Break, time=0.1, value=1}
         <at 1 0>
        ]}}}
     Scale Curve=10
    )
   )
  ]
  rendererClasses [&NiagaraSpriteRendererProperties, &NiagaraLightRendererProperties]
  renderers [
   {rendererClass=&NiagaraSpriteRendererProperties}
   {rendererClass=&NiagaraLightRendererProperties, rendererIndex=1}
   {rendererClass=&NiagaraSpriteRendererProperties, rendererIndex=2}
  ]
 }
 Embers {
  emitterUpdateScript [
   &EmitterState()
   &SpawnBurst_Instantaneous(Spawn Count=30, Use Spawn Probability=false, UseLoopCountLimit=false)
  ]
  particleSpawnScript [
   &PartitionParticles(Number Of Partitions=4)
   /SetVariables(Particles.Random size:NiagaraFloat=&RandomRangeFloat(Minimum=0.1, Maximum=0.2)) {isSetParameters=true}
   &InitializeParticle(
    Color Mode=::NewEnumerator0
    Lifetime Max=0.7
    Lifetime Min=0.55
    Lifetime Mode=::NewEnumerator1
    Sprite Rotation Mode=::NewEnumerator2
    Sprite UV Mode=::NewEnumerator3
    UsePositionOffset=false
    Write Material Random=true
   ) {module=InitializeParticle001}
   &DynamicMaterialParameters(
    Index 0 Param 2=&RandomRangeFloat(Minimum=0, Maximum=4)
    Index 0 Param 4=2.5
    Module.Param0WriteEnabled=true
    Module.Param1WriteEnabled=true
    Module.Param2WriteEnabled=false
    Module.Param3WriteEnabled=true
   ) {module=DynamicMaterialParameters001}
   &SkeletalMeshLocation(
    Bone Sampling Mode=::NewEnumerator0
    MaskSpawnGroup=false
    OverrideRandomSeed=false
    Position Sampling=::NewEnumerator1
    Sampled Position Offset=<-7 0 0>
    Skeletal Mesh=$System.Skeletal Mesh
   )
   &ShapeLocation(MaskSpawnGroup=false, Override Seed=false, Sphere Radius=15, UseOwnerScale=true)
   &AddVelocity(Velocity=&RandomRangeVector(Minimum=<-250 0 0>, Maximum=<-400 0 0>)) {module=AddVelocity001}
   &CameraOffset(Camera Offset Amount=-50) {module=CameraOffset001}
   /SetVariables() {isSetParameters=true}
   /SetVariables(Particles.DY3:Vector4f=&MakeVector4(X=3, Y=1, Z=1, W=1)) {isSetParameters=true}
   /SetVariables(Particles.Position=&AddVectorToPosition(Position=$Particles.Position, Vector=<0 0 15>)) {isSetParameters=true}
  ]
  particleUpdateScript [
   /SetVariables(Particles.Size Scale:NiagaraFloat=&FloatFromCurve(FloatCurve=DataInterface() {propertyValues {Curve {keys [<at 0 0>,<at 0.1 0.5>,<at 1 0>]}}})) {isSetParameters=true}
   &SubUVAnimation(End Frame=11, Start Frame=8, SubUV Animation Mode:ENiagara_SubUVLookupMode=::NewEnumerator2)
   /Niagara/Modules/Update/Forces/CurlNoiseForce(Noise Frequency=100, Noise Strength=&Multiply_Float(A=50, B=100), Pan Noise=true) {enabled=false}
   &ParticleState()
   &ScaleSpriteSize(Scale Sprite Size Mode=::NewEnumerator0, Uniform Scale Factor=$Particles.Size Scale) {module=ScaleSpriteSize001}
   &SpriteRotationRate(Rotation Rate=&RandomRangeFloat(Minimum=-5, Maximum=5)) {module=SpriteRotationRate001}
   &Drag(Drag=4, Use Linear Drag=true, Use Rotational Drag=true)
   &SolveForcesAndVelocity(Clamp Velocity=false, Limit Acceleration=false)
   &FN_Sprite_Attributes(
    Color=&MakeLinearColorFromVectorAndFloat(
     "Vector (RGB)"=&MakeVectorFromLinearColorRGB(Color=&ColorFromCurve(ColorCurve=DataInterface() {propertyValues {
         AlphaCurve {keys []}
         BlueCurve {keys [<at 0.0072 3.886>, {arriveTangent=-3.5308, interpMode=RCIM_Cubic, leaveTangent=-3.5308, time=0.5363, value=0.8197}, <at 1 0.3808 RCIM_Cubic>]}
         GreenCurve {keys [
           <at 0.0072 6.2369>
           {arriveTangent=-5.0406, interpMode=RCIM_Cubic, leaveTangent=-5.0406, time=0.5363, value=1.8117}
           <at 1 1.2327 RCIM_Cubic>
          ]}
         RedCurve {keys [<at 0.0072 10>, {arriveTangent=-7.051, interpMode=RCIM_Cubic, leaveTangent=-7.051, time=0.5363, value=3}, <at 1 3 RCIM_Cubic>]}
        }}))
     "Float (Alpha)"=2
    )
    Create Color Attribute=true
    Create SpriteSize Attribute=true
    SpriteSize=&Vector2DFromFloat(Value=&Multiply_Float(A=100, B=&Multiply_Float(A=$Particles.Random size, B=$Particles.Size Scale)))
    Write to Module Namespace=true
   )
   &FN_Sprite_Attributes(
    Color=&MakeLinearColorFromVectorAndFloat(
     "Vector (RGB)"=&MakeVectorFromLinearColorRGB(Color=<rgba 1.5 0.78 0.2559 1>)
     "Float (Alpha)"=&Multiply_Float(
      A=&FN_FrontEndFloatModifier(Gameplay=0.3, Front End=0.2)
      B=&FloatFromCurve(
       FloatCurve=DataInterface() {propertyValues {Curve {keys [{interpMode=RCIM_Cubic, tangentMode=RCTM_User, time=0, value=1}, {interpMode=RCIM_Cubic, tangentMode=RCTM_User, time=1, value=1}]}}}
       Scale Curve=0.5
      )
     )
    )
    Create Color Attribute=true
    Create SpriteSize Attribute=true
    SpriteSize=&Vector2DFromFloat(Value=&Multiply_Float(A=350, B=&Multiply_Float(A=$Particles.Random size, B=$Particles.Size Scale)))
    Write to Module Namespace=true
   ) {module=FN_Sprite_Attributes001}
  ]
  rendererClasses [&NiagaraSpriteRendererProperties]
  renderers [
   {rendererClass=&NiagaraSpriteRendererProperties}
   {rendererClass=&NiagaraSpriteRendererProperties, rendererIndex=1}
   {rendererClass=&NiagaraSpriteRendererProperties, rendererIndex=2}
  ]
 }
 Flare {
  emitterUpdateScript [
   &EmitterState()
   &SpawnBurst_Instantaneous(Spawn Count=1, Spawn Time=$System.BurstDelay, Use Spawn Probability=false)
  ]
  particleSpawnScript [
   &InitializeParticle(
    Color=<rgba 1 0.6398 0.1534 1>
    Lifetime=0.35
    Sprite Size Mode=::NewEnumerator3
    Uniform Sprite Size=600
    UsePositionOffset=false
    Write Material Random=true
   )
   &DynamicMaterialParameters(
    Index 0 Param 1=&RandomRangeFloat(Minimum=0.4, Maximum=0.6)
    Index 0 Param 2=1
    Index 0 Param 3=1
    Index 0 Param 4=1
    Module.Param0WriteEnabled=true
    Module.Param1WriteEnabled=true
    Module.Param2WriteEnabled=true
    Module.Param3WriteEnabled=true
   )
   &CameraOffset(Camera Offset Amount=50)
  ]
  particleUpdateScript [
   &SystemLocation(MaskSpawnGroup=false, Offset=&Lerp_Vector(A=<0 0 20>, B=<0 0 80>, Alpha=$Particles.NormalizedAge))
   &ParticleState()
   &SolveForcesAndVelocity(Clamp Velocity=false, Limit Acceleration=false)
   &ScaleColor(
    Scale Alpha=&Multiply_Float(
     A=1
     B=&FloatFromCurve(FloatCurve=DataInterface() {propertyValues {Curve {keys [<at 0 0>,<at 0.2 1>,<at 0.8 1>,<at 1 0>]}}})
    )
    ScaleA=true
    ScaleRGB=true
   )
   &ScaleSpriteSize(Scale Factor=&Vector2DFromFloat(Value=&FloatFromCurve(FloatCurve=DataInterface() {propertyValues {Curve {keys [<at 0 1.5 RCIM_Cubic>,<at 1 0>]}}})))
   /SetVariables(Particles.Flash size:Vector2f=&Multiply_Vector2D_ByFloat(
     Float=&Multiply_Float(
      A=400
      B=&FloatFromCurve(FloatCurve=DataInterface() {propertyValues {Curve {keys [
           {arriveTangent=4.1785, interpMode=RCIM_Cubic, leaveTangent=4.1785, tangentMode=RCTM_User, time=0, value=0}
           {arriveTangent=0.9995, interpMode=RCIM_Cubic, leaveTangent=0.9995, tangentMode=RCTM_User, time=0.1378, value=0.6825}
           {arriveTangent=-0.0405, interpMode=RCIM_Cubic, leaveTangent=-0.0405, tangentMode=RCTM_User, time=0.62, value=1}
          ]}}})
     )
     Vector2D=<1 1>
    )) {isSetParameters=true}
   /SetVariables(Particles.Flash color:LinearColor=&MakeLinearColorFromVectorAndFloat(
     "Vector (RGB)"=&MakeVectorFromLinearColorRGB(Color=<rgba 1 0.5437 0.0852 1>)
     "Float (Alpha)"=&Multiply_Float(
      A=&Multiply_Float(
       A=1
       B=&FloatFromCurve(FloatCurve=DataInterface() {propertyValues {Curve {keys [<at 0 0>,<at 0.124 1>,<at 0.496 1>,<at 0.62 0>]}}})
      )
      B=1
     )
    )) {isSetParameters=true}
  ]
  rendererClasses [&NiagaraSpriteRendererProperties]
  renderers [
   {rendererClass=&NiagaraSpriteRendererProperties}
   {rendererClass=&NiagaraSpriteRendererProperties, rendererIndex=1}
  ]
 }
 MeshFlames {
  emitterUpdateScript [
   &EmitterState()
   &SpawnBurst_Instantaneous(
    Spawn Count=&RandomRangeInt(Minimum=8, Maximum=10)
    Use Spawn Probability=false
    UseLoopCountLimit=false
   )
   &SpawnBurst_Instantaneous(Spawn Group=1, Use Spawn Probability=false, UseLoopCountLimit=false) {module=SpawnBurst_Instantaneous001}
  ]
  particleSpawnScript [
   /SetVariables(Particles.VisibilityTag=&SpawnGroup()) {isSetParameters=true}
   &InitializeParticle(
    AdjustAlpha=false
    AdjustHue=true
    AdjustSaturation=false
    AdjustValue=true
    Color=<rgba 0.7656 0.4726 0.252 1>
    Color Mode=::NewEnumerator3
    Lifetime=&MaskFloatBySpawnGroup(Masked Float=&RandomRangeFloat(Minimum=0.45, Maximum=0.9), Passthrough Float=0.5)
    Mesh Scale=&MaskVectorBySpawnGroup(Masked Vector=&RandomRangeVector(Minimum=<1 1 0.2>, Maximum=<1 1 0.6>), Passthrough Vector=<2 1 1>)
    Mesh Scale Mode=::NewEnumerator4
    Position Offset=<20 0 15>
    UsePositionOffset=true
    Value Range:Vector2f=<0.8 5>
    Write Material Random=true
   )
   &SkeletalMeshLocation(
    Bone Sampling Mode=::NewEnumerator0
    MaskSpawnGroup=false
    Orientation Sampling=::NewEnumerator1
    OverrideRandomSeed=false
    Skeletal Mesh=$System.Skeletal Mesh
    Write To Particle Position:NiagaraBool=/Niagara/DynamicInputs/Helpers/MaskBoolBySpawnGroup(Masked Bool:NiagaraBool=true, Passthrough Bool:NiagaraBool=false, Spawn Group Mask:NiagaraInt32=0)
   )
   &InitialMeshOrientation(
    Mesh Orientation Mode=::NewEnumerator7
    Rotation=&MaskVectorBySpawnGroup(Masked Vector=<0 -0.25 0>, Passthrough Vector=<0 0 0>)
    Use Rotation Vector=true
   )
   /SetVariables(Particles.GlowColor=<rgba 1 0.692 0.3353 0.25>) {isSetParameters=true}
  ]
  particleUpdateScript [
   &ParticleState()
   &ScaleMeshSize(Scale Factor=&MaskVectorBySpawnGroup(
     Masked Vector=&VectorFromCurve(VectorCurve=DataInterface() {propertyValues {XCurve {keys [<at 0 1>]}, YCurve {keys [<at 0 1>]}, ZCurve {keys [<at 0 0.3>,<at 0.2 0.75>,<at 1 1.2>]}}})
     Passthrough Vector=&VectorFromCurve(VectorCurve=DataInterface() {propertyValues {XCurve {keys [<at 0 0.35>,<at 0.2 0.85>,<at 1 1.425>]}, YCurve {keys [<at 0 1>]}, ZCurve {keys [<at 0 1>]}}})
    ))
   &DynamicMaterialParameters(
    Index 0 Param 1=1
    Index 0 Param 2=&RandomRangeFloat(Minimum=-5, Maximum=10)
    Module.Index0Param0Write=true
    Module.Index0Param1Write=true
    Module.Index0Param2Write=true
    Module.Index0Param3Write=true
   )
   &ScaleColor(
    Scale Alpha=&FloatFromCurve(FloatCurve=DataInterface() {propertyValues {Curve {keys [{interpMode=RCIM_Cubic, leaveTangent=-2.0227, tangentMode=RCTM_Break, time=0, value=1}, <at 1 0>]}}})
    Scale RGB=<1 1 2>
    ScaleA=true
    ScaleRGB=true
   )
   /SetVariables(Particles.GlowColor=&Multiply_LinearColorByFloat(
     Linear Color=$Particles.Initial.GlowColor
     Float=&FloatFromCurve(FloatCurve=DataInterface() {propertyValues {Curve {keys [
          {arriveTangent=1.2596, interpMode=RCIM_Cubic, leaveTangent=1.2596, tangentMode=RCTM_User, time=0, value=0.4}
          {arriveTangent=-0.4744, interpMode=RCIM_Cubic, leaveTangent=-3.8835, tangentMode=RCTM_Break, time=0.4, value=1}
          {arriveTangent=-0.0851, tangentMode=RCTM_User, time=1, value=0}
         ]}}})
    )) {isSetParameters=true}
  ]
  rendererClasses [&NiagaraMeshRendererProperties]
  renderers [
   {rendererClass=&NiagaraMeshRendererProperties, rendererIndex=0}
   {rendererClass=&NiagaraMeshRendererProperties, rendererIndex=1}
   {rendererClass=&NiagaraMeshRendererProperties, rendererIndex=2}
   {rendererClass=&NiagaraMeshRendererProperties, rendererIndex=3}
  ]
 }
 ParticleDecal {
  emitterUpdateScript [
   &EmitterState()
   &SpawnBurst_Instantaneous(Use Spawn Probability=false, UseLoopCountLimit=false)
  ]
  particleSpawnScript [
   &InitializeParticle(
    Lifetime=0.45
    Mesh Scale Mode=::NewEnumerator3
    Mesh Uniform Scale=2
    Position Offset=<0 0 10>
    Sprite Size Mode=::NewEnumerator3
    Uniform Sprite Size=150
    UsePositionOffset=true
    Write Material Random=true
   )
   &CameraOffset(Camera Offset Amount=30)
   &DynamicMaterialParameters(
    Index 0 Param 1=1
    Index 0 Param 3=1
    Module.Index0Param0Write=true
    Module.Index0Param1Write=true
    Module.Index0Param2Write=true
    Module.Index0Param3Write=true
   )
   &AddVelocity(Velocity=<-250 0 550>) {enabled=false}
   &ApplyInitialForces()
  ]
  particleUpdateScript [
   &ParticleState()
   &Drag(Drag=3, Use Linear Drag=true, Use Rotational Drag=true) {enabled=false}
   &SolveForcesAndVelocity(Clamp Velocity=false, Limit Acceleration=false)
   &ScaleColor(
    Linear Color Curve=DataInterface() {propertyValues {
      AlphaCurve {keys [<at 0.0663 1>,<at 1 0>]}
      BlueCurve {keys [<at 0.2748 5>,<at 0.5275 0.0405>]}
      GreenCurve {keys [<at 0.2748 3.3645>,<at 0.5275 1.1103>]}
      RedCurve {keys [<at 0.2748 0.7531>,<at 0.5275 3>]}
     }}
    Scale Mode=::NewEnumerator2
   )
  ]
  rendererClasses [&NiagaraSpriteRendererProperties]
  renderers [{rendererClass=&NiagaraSpriteRendererProperties}]
 }
 ShockWave {
  emitterUpdateScript [
   &EmitterState(Inactive Response=::NewEnumerator1, Life Cycle Mode=::NewEnumerator1, Loop Behavior=::NewEnumerator1, Loop Duration=0.8, UseLoopDelay=true)
   &SpawnBurst_Instantaneous(Spawn Count=1, Use Spawn Probability=false)
   &SpawnBurst_Instantaneous(Spawn Time=0.8, Use Spawn Probability=false, UseLoopCountLimit=false) {module=SpawnBurst_Instantaneous001}
  ]
  particleSpawnScript [
   /Niagara/Modules/Spawn/Initialization/InitializeParticle(
    Color=<rgba 10 1.61 0 2>
    Lifetime=0.2
    Mesh Scale=<1.3 1 1>
    Module.Write Color=true
    Module.Write Lifetime=true
    Module.Write Mass=true
    Module.Write Material Random=true
    Module.Write Position=false
    Module.Write Scale=true
    Module.Write SpriteRotation=false
    Module.Write SpriteSize=true
    Position=&SimulationPosition()
    Sprite Size=<5 5>
   ) {enabled=false}
   /Niagara/Modules/Spawn/Initialization/V2/InitializeParticle(
    Color=<rgba 10 5.638 0.4928 1.4>
    Lifetime=0.25
    Mass Mode=::NewEnumerator1
    Mesh Scale=<1.8 2 2>
    Mesh Scale Mode=::NewEnumerator4
    Position Offset=<10 0 0>
    Sprite Size=<5 5>
    Sprite Size Mode=::NewEnumerator4
    UsePositionOffset=true
    Write Material Random=true
   ) {module=InitializeParticle001}
   &InitialMeshOrientation(Rotation=<0 0 0.25>, Use Orientation Vector=true, Use Rotation Vector=true) {enabled=false}
   &AddVelocity(Velocity=<-100 0 0>)
  ]
  particleUpdateScript [
   &SkeletalMeshLocation(
    Bone / Socket Index=1
    Bone Sampling Mode=::NewEnumerator4
    MaskSpawnGroup=false
    Sampled Position Offset=$System.FXOffset
    Skeletal Mesh=$System.Skeletal Mesh
   ) {enabled=false, module=SkeletalMeshLocation001}
   &ParticleState()
   &Drag(Drag=2, Use Linear Drag=true, Use Rotational Drag=true)
   &SolveForcesAndVelocity(Clamp Velocity=false, Limit Acceleration=false)
   &ScaleColor(
    Scale Alpha=&Multiply_Float(
     A=1
     B=&Multiply_Float(
      A=&FloatFromCurve(FloatCurve=DataInterface() {propertyValues {Curve {keys [
           <at 0 0 RCIM_Cubic>
           {arriveTangent=-1.8703, interpMode=RCIM_Cubic, leaveTangent=-1.8703, tangentMode=RCTM_User, time=0.026, value=1}
           <at 1 0>
          ]}}})
      B=1
     )
    )
    Scale RGB=&MakeVectorFromLinearColorRGB(Color=&ColorFromCurve(ColorCurve=DataInterface() {propertyValues {
        AlphaCurve {keys []}
        BlueCurve {keys [{arriveTangent=2.3578, interpMode=RCIM_Cubic, time=0.1649, value=1}, <at 0.7697 1 RCIM_Cubic>, <at 0.9805 1>]}
        GreenCurve {keys [
          {arriveTangent=-30.4531, interpMode=RCIM_Cubic, time=0.1649, value=1}
          {arriveTangent=-1.1661, interpMode=RCIM_Cubic, leaveTangent=-1.1661, time=0.7697, value=0.3963}
          <at 0.9805 0.0489>
         ]}
        RedCurve {keys [
          {arriveTangent=-115.5308, interpMode=RCIM_Cubic, time=0.1649, value=1}
          {arriveTangent=-1.143, interpMode=RCIM_Cubic, leaveTangent=-1.143, time=0.7697, value=0.0824}
          <at 0.9805 0.0678>
         ]}
       }}))
    ScaleA=true
    ScaleRGB=true
   ) {module=ScaleColor001}
   &ScaleMeshSize(Scale Factor=&VectorFromCurve(
     VectorCurve=DataInterface() {propertyValues {XCurve {keys [<at 0 0.7>,<at 1 2>]}, YCurve {keys [<at 0 0.9>,<at 1 1.6>]}, ZCurve {keys [<at 0 0.9>,<at 1 2>]}}}
     Scale Curve=<0.6 0.6 0.6>
    ))
   &ScaleMeshSize(Scale Factor=&Multiply_VectorByFloat(Vector=<1 1 1>, Float=0.9)) {module=ScaleMeshSize001}
   &DynamicMaterialParameters(
    Index 0 Param 1=&FloatFromCurve(FloatCurve=DataInterface() {propertyValues {Curve {keys [{arriveTangent=0.2529, interpMode=RCIM_Cubic, leaveTangent=0.2529, tangentMode=RCTM_User, time=0, value=0}, <at 1 0.2 RCIM_Cubic>]}}})
    Index 0 Param 2=&RandomRangeFloat(Minimum=0, Maximum=1)
    Index 0 Param 3=50
    Index 0 Param 4=&FloatFromCurve(FloatCurve=DataInterface() {propertyValues {Curve {keys [
         {interpMode=RCIM_Cubic, leaveTangent=0.0377, tangentMode=RCTM_Break, time=0, value=0}
         {arriveTangent=-3.0374, interpMode=RCIM_Cubic, leaveTangent=2.0371, tangentMode=RCTM_Break, time=0.2, value=-1.5}
         {interpMode=RCIM_Cubic, tangentMode=RCTM_Break, time=1, value=0}
        ]}}})
    Module.Param0WriteEnabled=true
    Module.Param1WriteEnabled=true
    Module.Param2WriteEnabled=true
    Module.Param3WriteEnabled=true
   )
  ]
  rendererClasses [&NiagaraMeshRendererProperties]
  renderers [{rendererClass=&NiagaraMeshRendererProperties, rendererIndex=0}]
 }
 SpeedLines {
  emitterUpdateScript [
   &EmitterState()
   &SpawnBurst_Instantaneous(Spawn Count=12, Use Spawn Probability=false, UseLoopCountLimit=false)
  ]
  particleSpawnScript [
   &InitializeParticle(
    Color=<rgba 3 2.0091 0.3058 1.5>
    Lifetime Max=0.5
    Lifetime Min=0.3
    Lifetime Mode=::NewEnumerator1
    Sprite Size Max=<1.2 50>
    Sprite Size Min=<0.6 15>
    Sprite Size Mode=::NewEnumerator2
    UsePositionOffset=false
    Write Material Random=true
   )
   &SkeletalMeshLocation(MaskSpawnGroup=false, OverrideRandomSeed=false, Skeletal Mesh=$System.Skeletal Mesh) {enabled=false}
   &ShapeLocation(MaskSpawnGroup=false, Override Seed=false, Sphere Radius=70, Sphere Surface Distribution=0.6667, UseOwnerScale=true)
   &AddVelocity(Velocity=&RandomRangeVector(Minimum=<-200 0 0>, Maximum=<-500 0 0>))
  ]
  particleUpdateScript [
   &ParticleState()
   &ScaleColor(
    Linear Color Curve=DataInterface() {propertyValues {
      AlphaCurve {keys [<at 0 0>,<at 0.1573 1>,<at 0.5937 0.3>,<at 1 0>]}
      BlueCurve {keys [<at 0 1>,<at 0.8464 1>]}
      GreenCurve {keys [<at 0 1>,<at 0.8464 0.6297>]}
      RedCurve {keys [<at 0 1>,<at 0.8464 0.1334>]}
     }}
    Scale Mode=::NewEnumerator2
   )
   &CameraOffset(Camera Offset Amount=25)
   &Drag(Drag=3, Use Linear Drag=true, Use Rotational Drag=true)
   /Niagara/Modules/Update/Forces/V2/CurlNoiseForce(Bias Noise=false, Noise Frequency=20, Noise Strength=200, Pan Noise=false) {enabled=false}
   &SolveForcesAndVelocity(Clamp Velocity=false, Limit Acceleration=false)
   &ScaleSpriteSizeBySpeed(Max Scale Factor=<1.2 3>, Min Scale Factor=<1 1.4>)
  ]
  rendererClasses [&NiagaraSpriteRendererProperties]
  renderers [{rendererClass=&NiagaraSpriteRendererProperties}]
 }
}
summary {
 systemName=NS_Emote_FeatherMudLounge_SwapToLion_Burst_Gold
 emitters [
  {emitterName=ParticleDecal, rendererClasses [&NiagaraSpriteRendererProperties]}
  {emitterName=SpeedLines, rendererClasses [&NiagaraSpriteRendererProperties]}
  {emitterName=ShockWave, rendererClasses [&NiagaraMeshRendererProperties]}
  {emitterName=Flare, rendererClasses [&NiagaraSpriteRendererProperties]}
  {emitterName=Embers, rendererClasses [&NiagaraSpriteRendererProperties]}
  {emitterName=DarkGlow, rendererClasses [&NiagaraSpriteRendererProperties, &NiagaraLightRendererProperties]}
  {emitterName=MeshFlames, rendererClasses [&NiagaraMeshRendererProperties]}
 ]
 userVariables [
  User.bIsFrontEnd {type=NiagaraBool}
  User.bIsFrontEndPreview {type=NiagaraBool}
  User.VariantSwitch {type=NiagaraFloat, defaultValue=0, description=""}
 ]
}
